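# Sound cue for rifle bullet impacts on players in Athena, using the impacts sound class and bullet impact attenuation.
# A random node picks, with repeats allowed, among four rifle impact waves and pitch-enveloped versions of the first three, all equally weighted.
# The enveloped branches reuse the same wave players with different pitch curves, adding variation without extra assets.
# A root modulator varies the result, and a rifle impact concurrency set governs overlapping plays.
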SoundCue /Game/Sounds/Fort_Weapons/Impacts/Bullet_Impacts_S01/FNS01_Bullet_Player_Impact_Athena_Cue
root=Modulator_0
nodes [
 {
  id=Enveloper_0
  children [WavePlayer_0]
  properties {pitchCurve {editorCurveData {keys [{time=0}, <at 0.2252 1.4109>, <at 0.456 -0.1395>]}}, volumeCurve {editorCurveData {keys [<at 0 1>]}}}
 }
 {
  id=Enveloper_1
  children [WavePlayer_1]
  properties {
   pitchCurve {editorCurveData {keys [{time=0}, <at 0.254 0.6319>, <at 0.4 1.4972>, <at 0.7489 0.4609>]}}
   volumeCurve {editorCurveData {keys [<at 0 1>]}}
  }
 }
 {
  id=Enveloper_2
  children [WavePlayer_2]
  properties {pitchCurve {editorCurveData {keys [{time=0}, <at 0.0771 1.4823>, <at 0.4081 1.3073>]}}, volumeCurve {editorCurveData {keys [<at 0 1>]}}}
 }
 {id=Modulator_0, children [Random_0], properties {volumeMax=1.5}}
 {
  id=Random_0
  children [WavePlayer_0,WavePlayer_1,WavePlayer_2,Enveloper_0,Enveloper_1,Enveloper_2,WavePlayer_3]
  properties {bRandomizeWithoutReplacement=false, weights [1,1,1,1,1,1,1]}
 }
 {id=WavePlayer_0, properties {soundWaveAssetPtr=/Game/Sounds/Fort_Weapons/Impacts/player_bulletimpact_rifle_01}}
 {id=WavePlayer_1, properties {soundWaveAssetPtr=/Game/Sounds/Fort_Weapons/Impacts/player_bulletimpact_rifle_02}}
 {id=WavePlayer_2, properties {soundWaveAssetPtr=/Game/Sounds/Fort_Weapons/Impacts/player_bulletimpact_rifle_03}}
 {id=WavePlayer_3, properties {soundWaveAssetPtr=/Game/Sounds/Fort_Weapons/Impacts/player_bulletimpact_rifle_04}}
]
properties {
 attenuationSettings=/Game/Sounds/Attenuation/Fort_Bullet_Impact_Att
 maxDistance=5300
 soundClassObject=&Fort_Impacts
 volumeMultiplier=1.416
 concurrencyOverrides {maxCount=4, resolutionRule=PreventNew}
 concurrencySet [&RifleBulletImpact_Conc]
}
