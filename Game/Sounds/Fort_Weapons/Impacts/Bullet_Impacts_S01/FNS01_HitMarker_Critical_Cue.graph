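# Sound cue for critical-hit marker feedback, routed through the instigator 2D hit notification sound class.
# It mixes a random metal cannon impact, pitched up an octave, with a briefly delayed second layer.
# That delayed layer pairs a random enemy flesh impact with a slightly louder critical notification wave.
# Hit-marker attenuation applies, and a rifle bullet impact concurrency set is assigned.
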SoundCue /Game/Sounds/Fort_Weapons/Impacts/Bullet_Impacts_S01/FNS01_HitMarker_Critical_Cue
root=Mixer_0
nodes [
 {id=Delay_6, children [Mixer_1], properties {delayMax=0.15, delayMin=0.15}}
 {id=Mixer_0, children [Modulator_1,Modulator_6], properties {inputVolume [1,1]}}
 {id=Mixer_1, children [Random_7,WavePlayer_4], properties {inputVolume [1,1.25]}}
 {id=Modulator_1, children [Random_8], properties {pitchMax=2, pitchMin=2}}
 {id=Modulator_6, children [Delay_6], properties {pitchMax=1, pitchMin=1}}
 {id=Random_7, children [WavePlayer_20, WavePlayer_21, WavePlayer_22, WavePlayer_23], properties {weights [1,1,1,1]}}
 {id=Random_8, children [WavePlayer_5,WavePlayer_6,WavePlayer_7,WavePlayer_8,WavePlayer_9], properties {weights [1,1,1,1,1]}}
 {id=WavePlayer_20, properties {soundWaveAssetPtr=/Game/Sounds/Fort_Impact_Sounds/Flesh/player_impact_enemy_stereo_01}}
 {id=WavePlayer_21, properties {soundWaveAssetPtr=/Game/Sounds/Fort_Impact_Sounds/Flesh/player_impact_enemy_stereo_02}}
 {id=WavePlayer_22, properties {soundWaveAssetPtr=/Game/Sounds/Fort_Impact_Sounds/Flesh/player_impact_enemy_stereo_03}}
 {id=WavePlayer_23, properties {soundWaveAssetPtr=/Game/Sounds/Fort_Impact_Sounds/Flesh/player_impact_enemy_stereo_04}}
 {id=WavePlayer_4, properties {soundWaveAssetPtr=/Game/Athena/Sounds/HitNotifications/sfx_hitmarker_notification_critical_layer}}
 {id=WavePlayer_5, properties {soundWaveAssetPtr=/Game/Sounds/Fort_Impact_Sounds/Demolition/Metal/Metal_Cannon_Impact01}}
 {id=WavePlayer_6, properties {soundWaveAssetPtr=/Game/Sounds/Fort_Impact_Sounds/Demolition/Metal/Metal_Cannon_Impact02}}
 {id=WavePlayer_7, properties {soundWaveAssetPtr=/Game/Sounds/Fort_Impact_Sounds/Demolition/Metal/Metal_Cannon_Impact03}}
 {id=WavePlayer_8, properties {soundWaveAssetPtr=/Game/Sounds/Fort_Impact_Sounds/Demolition/Metal/Metal_Cannon_Impact04}}
 {id=WavePlayer_9, properties {soundWaveAssetPtr=/Game/Sounds/Fort_Impact_Sounds/Demolition/Metal/Metal_Cannon_Impact05}}
]
properties {
 attenuationSettings=/Game/Athena/Sounds/HitNotifications/Att_HitMarker_Instigate_MS
 maxDistance=2097152
 soundClassObject=&Fort_2DHitNotifications_Instigate_PSM
 volumeMultiplier=1.062
 concurrencyOverrides {maxCount=4, resolutionRule=PreventNew}
 concurrencySet [&RifleBulletImpact_Conc]
}
