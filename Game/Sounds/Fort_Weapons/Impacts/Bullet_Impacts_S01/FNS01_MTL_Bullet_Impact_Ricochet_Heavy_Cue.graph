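# Sound cue for heavy metal bullet impact ricochets, using bullet impact attenuation, the impacts sound class and rifle impact concurrency.
# Each branch mixes an immediate random impact wave with a briefly, randomly delayed random layer wave, both modulated in pitch and volume.
# A quality-level root gives the first level six variations per layer, while the other three levels share a reduced branch with four per layer.
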SoundCue /Game/Sounds/Fort_Weapons/Impacts/Bullet_Impacts_S01/FNS01_MTL_Bullet_Impact_Ricochet_Heavy_Cue
root=QualityLevel_4
nodes [
 {id=Delay_8, children [Modulator_17], properties {delayMax=0.15, delayMin=0.05}}
 {id=Delay_9, children [Modulator_19], properties {delayMax=0.15, delayMin=0.05}}
 {id=Mixer_8, children [Delay_8,Modulator_16], properties {inputVolume [1,1]}}
 {id=Mixer_9, children [Delay_9,Modulator_18], properties {inputVolume [1,1]}}
 {id=Modulator_16, children [Random_16], properties {pitchMax=1.35, pitchMin=1, volumeMax=1.5}}
 {id=Modulator_17, children [Random_17], properties {pitchMax=1.35, pitchMin=1, volumeMax=1.5}}
 {id=Modulator_18, children [Random_18], properties {pitchMax=1.35, pitchMin=1, volumeMax=1.5}}
 {id=Modulator_19, children [Random_19], properties {pitchMax=1.35, pitchMin=1, volumeMax=1.5}}
 {id=QualityLevel_4, children [Mixer_8,Mixer_9,Mixer_9,Mixer_9]}
 {
  id=Random_16
  children [WavePlayer_17, WavePlayer_18, WavePlayer_19, WavePlayer_20, WavePlayer_21, WavePlayer_22]
  properties {weights [1,1,1,1,1,1]}
 }
 {
  id=Random_17
  children [WavePlayer_11, WavePlayer_12, WavePlayer_13, WavePlayer_14, WavePlayer_15, WavePlayer_16]
  properties {weights [1,1,1,1,1,1]}
 }
 {id=Random_18, children [WavePlayer_30, WavePlayer_33, WavePlayer_74, WavePlayer_75], properties {weights [1,1,1,1]}}
 {id=Random_19, children [WavePlayer_24, WavePlayer_25, WavePlayer_27, WavePlayer_29], properties {weights [1,1,1,1]}}
 {id=WavePlayer_11, properties {soundWaveAssetPtr=/Game/Sounds/Fort_Weapons/Impacts/Bullet_Impacts_S01/Wavs/MTL_IMPACT_LAYER_HVY_000}}
 {id=WavePlayer_12, properties {soundWaveAssetPtr=/Game/Sounds/Fort_Weapons/Impacts/Bullet_Impacts_S01/Wavs/MTL_IMPACT_LAYER_HVY_001}}
 {id=WavePlayer_13, properties {soundWaveAssetPtr=/Game/Sounds/Fort_Weapons/Impacts/Bullet_Impacts_S01/Wavs/MTL_IMPACT_LAYER_HVY_002}}
 {id=WavePlayer_14, properties {soundWaveAssetPtr=/Game/Sounds/Fort_Weapons/Impacts/Bullet_Impacts_S01/Wavs/MTL_IMPACT_LAYER_HVY_003}}
 {id=WavePlayer_15, properties {soundWaveAssetPtr=/Game/Sounds/Fort_Weapons/Impacts/Bullet_Impacts_S01/Wavs/MTL_IMPACT_LAYER_HVY_004}}
 {id=WavePlayer_16, properties {soundWaveAssetPtr=/Game/Sounds/Fort_Weapons/Impacts/Bullet_Impacts_S01/Wavs/MTL_IMPACT_LAYER_HVY_005}}
 {id=WavePlayer_17, properties {soundWaveAssetPtr=/Game/Sounds/Fort_Weapons/Impacts/Bullet_Impacts_S01/Wavs/MTL_IMPACT_HVY_01}}
 {id=WavePlayer_18, properties {soundWaveAssetPtr=/Game/Sounds/Fort_Weapons/Impacts/Bullet_Impacts_S01/Wavs/MTL_IMPACT_HVY_02}}
 {id=WavePlayer_19, properties {soundWaveAssetPtr=/Game/Sounds/Fort_Weapons/Impacts/Bullet_Impacts_S01/Wavs/MTL_IMPACT_HVY_03}}
 {id=WavePlayer_20, properties {soundWaveAssetPtr=/Game/Sounds/Fort_Weapons/Impacts/Bullet_Impacts_S01/Wavs/MTL_IMPACT_HVY_04}}
 {id=WavePlayer_21, properties {soundWaveAssetPtr=/Game/Sounds/Fort_Weapons/Impacts/Bullet_Impacts_S01/Wavs/MTL_IMPACT_HVY_05}}
 {id=WavePlayer_22, properties {soundWaveAssetPtr=/Game/Sounds/Fort_Weapons/Impacts/Bullet_Impacts_S01/Wavs/MTL_IMPACT_HVY_06}}
 {id=WavePlayer_24, properties {soundWaveAssetPtr=&MTL_IMPACT_LAYER_HVY_000}}
 {id=WavePlayer_25, properties {soundWaveAssetPtr=&MTL_IMPACT_LAYER_HVY_001}}
 {id=WavePlayer_27, properties {soundWaveAssetPtr=&MTL_IMPACT_LAYER_HVY_003}}
 {id=WavePlayer_29, properties {soundWaveAssetPtr=&MTL_IMPACT_LAYER_HVY_005}}
 {id=WavePlayer_30, properties {soundWaveAssetPtr=&MTL_IMPACT_HVY_01}}
 {id=WavePlayer_33, properties {soundWaveAssetPtr=&MTL_IMPACT_HVY_02}}
 {id=WavePlayer_74, properties {soundWaveAssetPtr=&MTL_IMPACT_HVY_05}}
 {id=WavePlayer_75, properties {soundWaveAssetPtr=&MTL_IMPACT_HVY_06}}
]
properties {
 attenuationSettings=/Game/Sounds/Attenuation/Fort_Bullet_Impact_Att
 maxDistance=5300
 soundClassObject=&Fort_Impacts
 volumeMultiplier=1.416
 concurrencySet [&RifleBulletImpact_Conc]
}
