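# Sound cue for light metal bullet impacts (named as a ricochet), spatialized with bullet impact attenuation and limited by rifle bullet impact concurrency.
# A quality-level root sends the first level to a random pick of eight impact waves and the other three levels to a shared four-wave subset.
# Each random branch passes through a modulator that varies pitch and volume to reduce repetition.
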SoundCue /Game/Sounds/Fort_Weapons/Impacts/Bullet_Impacts_S01/FNS01_MTL_Bullet_Impact_Ricochet_Light_Cue
root=QualityLevel_0
nodes [
 {id=Modulator_0, children [Random_0], properties {pitchMax=1.1, volumeMax=1.5}}
 {id=Modulator_1, children [Random_1], properties {pitchMax=1.1, volumeMax=1.5}}
 {id=QualityLevel_0, children [Modulator_0,Modulator_1,Modulator_1,Modulator_1]}
 {
  id=Random_0
  children [WavePlayer_18, WavePlayer_19, WavePlayer_20, WavePlayer_22, WavePlayer_26, WavePlayer_27, WavePlayer_28, WavePlayer_29]
  properties {weights [1,1,1,1,1,1,1,1]}
 }
 {id=Random_1, children [WavePlayer_31, WavePlayer_32, WavePlayer_44, WavePlayer_45], properties {weights [1,1,1,1]}}
 {id=WavePlayer_18, properties {soundWaveAssetPtr=/Game/Sounds/Fort_Weapons/Impacts/Bullet_Impacts_S01/Wavs/MTL_IMPACT_LIGHT_01}}
 {id=WavePlayer_19, properties {soundWaveAssetPtr=/Game/Sounds/Fort_Weapons/Impacts/Bullet_Impacts_S01/Wavs/MTL_IMPACT_LIGHT_02}}
 {id=WavePlayer_20, properties {soundWaveAssetPtr=/Game/Sounds/Fort_Weapons/Impacts/Bullet_Impacts_S01/Wavs/MTL_IMPACT_LIGHT_03}}
 {id=WavePlayer_22, properties {soundWaveAssetPtr=/Game/Sounds/Fort_Weapons/Impacts/Bullet_Impacts_S01/Wavs/MTL_IMPACT_LIGHT_04}}
 {id=WavePlayer_26, properties {soundWaveAssetPtr=/Game/Sounds/Fort_Weapons/Impacts/Bullet_Impacts_S01/Wavs/MTL_IMPACT_LIGHT_05}}
 {id=WavePlayer_27, properties {soundWaveAssetPtr=/Game/Sounds/Fort_Weapons/Impacts/Bullet_Impacts_S01/Wavs/MTL_IMPACT_LIGHT_06}}
 {id=WavePlayer_28, properties {soundWaveAssetPtr=/Game/Sounds/Fort_Weapons/Impacts/Bullet_Impacts_S01/Wavs/MTL_IMPACT_LIGHT_07}}
 {id=WavePlayer_29, properties {soundWaveAssetPtr=/Game/Sounds/Fort_Weapons/Impacts/Bullet_Impacts_S01/Wavs/MTL_IMPACT_LIGHT_08}}
 {id=WavePlayer_31, properties {soundWaveAssetPtr=&MTL_IMPACT_LIGHT_01}}
 {id=WavePlayer_32, properties {soundWaveAssetPtr=&MTL_IMPACT_LIGHT_02}}
 {id=WavePlayer_44, properties {soundWaveAssetPtr=&MTL_IMPACT_LIGHT_04}}
 {id=WavePlayer_45, properties {soundWaveAssetPtr=&MTL_IMPACT_LIGHT_07}}
]
properties {
 attenuationSettings=/Game/Sounds/Attenuation/Fort_Bullet_Impact_Att
 maxDistance=5300
 soundClassObject=&Fort_Impacts
 volumeMultiplier=1.416
 concurrencySet [&RifleBulletImpact_Conc]
}
