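# Sound cue for medium metal bullet impacts (named as a ricochet), playing a random MTL_IMPACT_MED wave.
# A quality-level root sends its first tier to a six-wave random pool and the other three tiers to a shared, trimmed four-wave pool.
# Each pool passes through a modulator for slight pitch and volume variation.
# Bullet-impact attenuation, the impacts sound class and rifle-impact concurrency govern playback.
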SoundCue /Game/Sounds/Fort_Weapons/Impacts/Bullet_Impacts_S01/FNS01_MTL_Bullet_Impact_Ricochet_Med_Cue
root=QualityLevel_1
nodes [
 {id=Modulator_1, children [Random_2], properties {pitchMin=0.98, volumeMax=1.5}}
 {id=Modulator_3, children [Random_3], properties {pitchMin=0.98, volumeMax=1.5}}
 {id=QualityLevel_1, children [Modulator_1,Modulator_3,Modulator_3,Modulator_3]}
 {
  id=Random_2
  children [WavePlayer_33, WavePlayer_34, WavePlayer_35, WavePlayer_36, WavePlayer_37, WavePlayer_38]
  properties {weights [1,1,1,1,1,1]}
 }
 {id=Random_3, children [WavePlayer_39, WavePlayer_40, WavePlayer_41, WavePlayer_42], properties {weights [1,1,1,1]}}
 {id=WavePlayer_33, properties {soundWaveAssetPtr=/Game/Sounds/Fort_Weapons/Impacts/Bullet_Impacts_S01/Wavs/MTL_IMPACT_MED_01}}
 {id=WavePlayer_34, properties {soundWaveAssetPtr=/Game/Sounds/Fort_Weapons/Impacts/Bullet_Impacts_S01/Wavs/MTL_IMPACT_MED_02}}
 {id=WavePlayer_35, properties {soundWaveAssetPtr=/Game/Sounds/Fort_Weapons/Impacts/Bullet_Impacts_S01/Wavs/MTL_IMPACT_MED_03}}
 {id=WavePlayer_36, properties {soundWaveAssetPtr=/Game/Sounds/Fort_Weapons/Impacts/Bullet_Impacts_S01/Wavs/MTL_IMPACT_MED_04}}
 {id=WavePlayer_37, properties {soundWaveAssetPtr=/Game/Sounds/Fort_Weapons/Impacts/Bullet_Impacts_S01/Wavs/MTL_IMPACT_MED_05}}
 {id=WavePlayer_38, properties {soundWaveAssetPtr=/Game/Sounds/Fort_Weapons/Impacts/Bullet_Impacts_S01/Wavs/MTL_IMPACT_MED_06}}
 {id=WavePlayer_39, properties {soundWaveAssetPtr=&MTL_IMPACT_MED_01}}
 {id=WavePlayer_40, properties {soundWaveAssetPtr=&MTL_IMPACT_MED_02}}
 {id=WavePlayer_41, properties {soundWaveAssetPtr=&MTL_IMPACT_MED_04}}
 {id=WavePlayer_42, properties {soundWaveAssetPtr=&MTL_IMPACT_MED_05}}
]
properties {
 attenuationSettings=/Game/Sounds/Attenuation/Fort_Bullet_Impact_Att
 maxDistance=5300
 soundClassObject=&Fort_Impacts
 volumeMultiplier=1.416
 concurrencySet [&RifleBulletImpact_Conc]
}
