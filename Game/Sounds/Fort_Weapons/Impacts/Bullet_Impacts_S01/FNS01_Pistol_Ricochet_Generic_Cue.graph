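# Sound cue for generic pistol bullet ricochet impacts, spatialized with bullet-impact attenuation and assigned to the impacts sound class and rifle bullet-impact concurrency.
# A quality-level switch feeds branches that each apply a short random delay and modulation, then randomly pick an equally weighted ricochet wave.
# The first quality level draws from six waves, while the other three share one branch with a four-wave subset.
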SoundCue /Game/Sounds/Fort_Weapons/Impacts/Bullet_Impacts_S01/FNS01_Pistol_Ricochet_Generic_Cue
root=QualityLevel_2
nodes [
 {id=Delay_4, children [Modulator_4], properties {delayMax=0.15, delayMin=0.05}}
 {id=Delay_5, children [Modulator_5], properties {delayMax=0.15, delayMin=0.05}}
 {id=Modulator_4, children [Random_9], properties {volumeMax=1.5}}
 {id=Modulator_5, children [Random_5], properties {volumeMax=1.5}}
 {id=QualityLevel_2, children [Delay_4,Delay_5,Delay_5,Delay_5]}
 {id=Random_5, children [WavePlayer_13, WavePlayer_12, WavePlayer_11, WavePlayer_10], properties {weights [1,1,1,1]}}
 {id=Random_9, children [WavePlayer_4,WavePlayer_5,WavePlayer_6,WavePlayer_7,WavePlayer_8,WavePlayer_9], properties {weights [1,1,1,1,1,1]}}
 {id=WavePlayer_10, properties {soundWaveAssetPtr=/Game/Sounds/Fort_Weapons/Impacts/Bullet_Impacts_S01/Wavs/new_generic_ricochet_01}}
 {id=WavePlayer_11, properties {soundWaveAssetPtr=/Game/Sounds/Fort_Weapons/Impacts/Bullet_Impacts_S01/Wavs/new_generic_ricochet_02}}
 {id=WavePlayer_12, properties {soundWaveAssetPtr=/Game/Sounds/Fort_Weapons/Impacts/Bullet_Impacts_S01/Wavs/new_generic_ricochet_05}}
 {id=WavePlayer_13, properties {soundWaveAssetPtr=/Game/Sounds/Fort_Weapons/Impacts/Bullet_Impacts_S01/Wavs/new_generic_ricochet_06}}
 {id=WavePlayer_4, properties {soundWaveAssetPtr=&new_generic_ricochet_06}}
 {id=WavePlayer_5, properties {soundWaveAssetPtr=&new_generic_ricochet_05}}
 {id=WavePlayer_6, properties {soundWaveAssetPtr=/Game/Sounds/Fort_Weapons/Impacts/Bullet_Impacts_S01/Wavs/new_generic_ricochet_04}}
 {id=WavePlayer_7, properties {soundWaveAssetPtr=/Game/Sounds/Fort_Weapons/Impacts/Bullet_Impacts_S01/Wavs/new_generic_ricochet_03}}
 {id=WavePlayer_8, properties {soundWaveAssetPtr=&new_generic_ricochet_02}}
 {id=WavePlayer_9, properties {soundWaveAssetPtr=&new_generic_ricochet_01}}
]
properties {
 attenuationSettings=/Game/Sounds/Attenuation/Fort_Bullet_Impact_Att
 maxDistance=5300
 soundClassObject=&Fort_Impacts
 volumeMultiplier=1.416
 concurrencyOverrides {maxCount=4, resolutionRule=PreventNew}
 concurrencySet [&RifleBulletImpact_Conc]
}
