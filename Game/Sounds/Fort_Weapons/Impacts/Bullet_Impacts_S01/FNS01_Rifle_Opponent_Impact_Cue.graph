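# Sound cue for rifle bullet impacts on opponents, using the impact attenuation, sound class and shared rifle impact concurrency.
# A quality-level node sends the first level to a random pick of eight impact waves.
# The other three levels share a random pick of five of those waves.
# Both branches pass through identically configured modulators that vary pitch and volume per play.
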SoundCue /Game/Sounds/Fort_Weapons/Impacts/Bullet_Impacts_S01/FNS01_Rifle_Opponent_Impact_Cue
root=QualityLevel_0
nodes [
 {id=Modulator_1, children [Random_1], properties {pitchMax=1.2, pitchMin=0.9, volumeMax=1.5}}
 {id=Modulator_2, children [Random_2], properties {pitchMax=1.2, pitchMin=0.9, volumeMax=1.5}}
 {id=QualityLevel_0, children [Modulator_1,Modulator_2,Modulator_2,Modulator_2]}
 {
  id=Random_1
  children [WavePlayer_4, WavePlayer_5, WavePlayer_6, WavePlayer_7, WavePlayer_8, WavePlayer_9, WavePlayer_10, WavePlayer_11]
  properties {weights [1,1,1,1,1,1,1,1]}
 }
 {id=Random_2, children [WavePlayer_12, WavePlayer_13, WavePlayer_14, WavePlayer_15, WavePlayer_16], properties {weights [1,1,1,1,1]}}
 {id=WavePlayer_10, properties {soundWaveAssetPtr=/Game/Sounds/Fort_Weapons/Impacts/opponent_bullet_impact_07}}
 {id=WavePlayer_11, properties {soundWaveAssetPtr=/Game/Sounds/Fort_Weapons/Impacts/opponent_bullet_impact_08}}
 {id=WavePlayer_12, properties {soundWaveAssetPtr=/Game/Sounds/Fort_Weapons/Impacts/opponent_bullet_impact_01}}
 {id=WavePlayer_13, properties {soundWaveAssetPtr=/Game/Sounds/Fort_Weapons/Impacts/opponent_bullet_impact_02}}
 {id=WavePlayer_14, properties {soundWaveAssetPtr=/Game/Sounds/Fort_Weapons/Impacts/opponent_bullet_impact_03}}
 {id=WavePlayer_15, properties {soundWaveAssetPtr=&opponent_bullet_impact_07}}
 {id=WavePlayer_16, properties {soundWaveAssetPtr=&opponent_bullet_impact_08}}
 {id=WavePlayer_4, properties {soundWaveAssetPtr=&opponent_bullet_impact_01}}
 {id=WavePlayer_5, properties {soundWaveAssetPtr=&opponent_bullet_impact_02}}
 {id=WavePlayer_6, properties {soundWaveAssetPtr=&opponent_bullet_impact_03}}
 {id=WavePlayer_7, properties {soundWaveAssetPtr=/Game/Sounds/Fort_Weapons/Impacts/opponent_bullet_impact_04}}
 {id=WavePlayer_8, properties {soundWaveAssetPtr=/Game/Sounds/Fort_Weapons/Impacts/opponent_bullet_impact_05}}
 {id=WavePlayer_9, properties {soundWaveAssetPtr=/Game/Sounds/Fort_Weapons/Impacts/opponent_bullet_impact_06}}
]
properties {
 attenuationSettings=/Game/Sounds/Attenuation/Fort_Bullet_Impact_Att
 maxDistance=5300
 soundClassObject=&Fort_Impacts
 volumeMultiplier=1.416
 concurrencyOverrides {maxCount=4, resolutionRule=PreventNew}
 concurrencySet [&RifleBulletImpact_Conc]
}
