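# Sound cue for heavy bullet impacts on stone with occasional ricochets, using bullet-impact attenuation and rifle impact concurrency.
# Each mix layers a random rock hit, an optional ricochet and a heavy detail layer, with randomized pitch and volume and short random delays on the latter two.
# A weighted empty branch silences the ricochet more often than not.
# A quality node gives the first level the full wave pools and the other three a shared, smaller set.
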SoundCue /Game/Sounds/Fort_Weapons/Impacts/Bullet_Impacts_S01/FNS01_STONE_Bullet_Impact_Ricochet_Heavy_Cue
root=QualityLevel_1
nodes [
 {id=Delay_3, children [Modulator_9], properties {delayMax=0.25}}
 {id=Delay_5, children [Modulator_10], properties {delayMax=0.15}}
 {id=Delay_6, children [Modulator_12], properties {delayMax=0.25}}
 {id=Delay_7, children [Modulator_13], properties {delayMax=0.15}}
 {id=Mixer_2, children [Modulator_1,Delay_3,Delay_5], properties {inputVolume [1,1,1]}}
 {id=Mixer_3, children [Modulator_11,Delay_6,Delay_7], properties {inputVolume [1,1,1]}}
 {id=Modulator_1, children [Random_12], properties {pitchMax=1.2, pitchMin=0.9, volumeMax=1.5}}
 {id=Modulator_10, children [Random_15], properties {pitchMax=1.2, pitchMin=0.9, volumeMax=1.5}}
 {id=Modulator_11, children [Random_16], properties {pitchMax=1.2, pitchMin=0.9, volumeMax=1.5}}
 {id=Modulator_12, children [Random_18], properties {pitchMax=1.2, pitchMin=0.9, volumeMax=1.5}}
 {id=Modulator_13, children [Random_19], properties {pitchMax=1.2, pitchMin=0.9, volumeMax=1.5}}
 {id=Modulator_9, children [Random_14], properties {pitchMax=1.2, pitchMin=0.9, volumeMax=1.5}}
 {id=QualityLevel_1, children [Mixer_2,Mixer_3,Mixer_3,Mixer_3]}
 {
  id=Random_12
  children [WavePlayer_27, WavePlayer_28, WavePlayer_29, WavePlayer_30, WavePlayer_33, WavePlayer_38]
  properties {weights [1,1,1,1,1,1]}
 }
 {id=Random_13, children [WavePlayer_41, WavePlayer_44, WavePlayer_45, WavePlayer_46, WavePlayer_47], properties {weights [1,1,1,1,1]}}
 {id=Random_14, children [Random_13,null], properties {bRandomizeWithoutReplacement=false, weights [0.672,1]}}
 {id=Random_15, children [WavePlayer_48, WavePlayer_51, WavePlayer_74, WavePlayer_75, WavePlayer_76], properties {weights [1,1,1,1,1]}}
 {id=Random_16, children [WavePlayer_77, WavePlayer_78, WavePlayer_79, WavePlayer_82], properties {weights [1,1,1,1]}}
 {id=Random_17, children [WavePlayer_83, WavePlayer_86, WavePlayer_87], properties {weights [1,1,1]}}
 {id=Random_18, children [Random_17,null], properties {bRandomizeWithoutReplacement=false, weights [0.672,1]}}
 {id=Random_19, children [WavePlayer_88, WavePlayer_89, WavePlayer_90], properties {weights [1,1,1]}}
 {id=WavePlayer_27, properties {soundWaveAssetPtr=/Game/Sounds/Fort_Weapons/Impacts/Bullet_Impacts_S01/Wavs/ROCK_IMPACT_HVY_02}}
 {id=WavePlayer_28, properties {soundWaveAssetPtr=/Game/Sounds/Fort_Weapons/Impacts/Bullet_Impacts_S01/Wavs/ROCK_IMPACT_HVY_03}}
 {id=WavePlayer_29, properties {soundWaveAssetPtr=/Game/Sounds/Fort_Weapons/Impacts/Bullet_Impacts_S01/Wavs/ROCK_IMPACT_HVY_04}}
 {id=WavePlayer_30, properties {soundWaveAssetPtr=/Game/Sounds/Fort_Weapons/Impacts/Bullet_Impacts_S01/Wavs/ROCK_IMPACT_HVY_05}}
 {id=WavePlayer_33, properties {soundWaveAssetPtr=/Game/Sounds/Fort_Weapons/Impacts/Bullet_Impacts_S01/Wavs/ROCK_IMPACT_HVY_06}}
 {id=WavePlayer_38, properties {soundWaveAssetPtr=/Game/Sounds/Fort_Weapons/Impacts/Bullet_Impacts_S01/Wavs/ROCK_IMPACT_HVY_08}}
 {id=WavePlayer_41, properties {soundWaveAssetPtr=/Game/Sounds/Fort_Weapons/Impacts/Bullet_Impacts_S01/Wavs/ROCK_IMPACT_RICO_030}}
 {id=WavePlayer_44, properties {soundWaveAssetPtr=/Game/Sounds/Fort_Weapons/Impacts/Bullet_Impacts_S01/Wavs/ROCK_IMPACT_RICO_032}}
 {id=WavePlayer_45, properties {soundWaveAssetPtr=/Game/Sounds/Fort_Weapons/Impacts/Bullet_Impacts_S01/Wavs/ROCK_IMPACT_RICO_033}}
 {id=WavePlayer_46, properties {soundWaveAssetPtr=/Game/Sounds/Fort_Weapons/Impacts/Bullet_Impacts_S01/Wavs/ROCK_IMPACT_RICO_034}}
 {id=WavePlayer_47, properties {soundWaveAssetPtr=/Game/Sounds/Fort_Weapons/Impacts/Bullet_Impacts_S01/Wavs/ROCK_IMPACT_RICO_035}}
 {id=WavePlayer_48, properties {soundWaveAssetPtr=/Game/Sounds/Fort_Weapons/Impacts/Bullet_Impacts_S01/Wavs/ROCK_IMPACT_LAYER_HVY_000}}
 {id=WavePlayer_51, properties {soundWaveAssetPtr=/Game/Sounds/Fort_Weapons/Impacts/Bullet_Impacts_S01/Wavs/ROCK_IMPACT_LAYER_HVY_001}}
 {id=WavePlayer_74, properties {soundWaveAssetPtr=/Game/Sounds/Fort_Weapons/Impacts/Bullet_Impacts_S01/Wavs/ROCK_IMPACT_LAYER_HVY_002}}
 {id=WavePlayer_75, properties {soundWaveAssetPtr=/Game/Sounds/Fort_Weapons/Impacts/Bullet_Impacts_S01/Wavs/ROCK_IMPACT_LAYER_HVY_003}}
 {id=WavePlayer_76, properties {soundWaveAssetPtr=/Game/Sounds/Fort_Weapons/Impacts/Bullet_Impacts_S01/Wavs/ROCK_IMPACT_LAYER_HVY_005}}
 {id=WavePlayer_77, properties {soundWaveAssetPtr=&ROCK_IMPACT_HVY_02}}
 {id=WavePlayer_78, properties {soundWaveAssetPtr=&ROCK_IMPACT_HVY_03}}
 {id=WavePlayer_79, properties {soundWaveAssetPtr=&ROCK_IMPACT_HVY_04}}
 {id=WavePlayer_82, properties {soundWaveAssetPtr=&ROCK_IMPACT_HVY_08}}
 {id=WavePlayer_83, properties {soundWaveAssetPtr=&ROCK_IMPACT_RICO_030}}
 {id=WavePlayer_86, properties {soundWaveAssetPtr=&ROCK_IMPACT_RICO_034}}
 {id=WavePlayer_87, properties {soundWaveAssetPtr=&ROCK_IMPACT_RICO_035}}
 {id=WavePlayer_88, properties {soundWaveAssetPtr=&ROCK_IMPACT_LAYER_HVY_000}}
 {id=WavePlayer_89, properties {soundWaveAssetPtr=&ROCK_IMPACT_LAYER_HVY_001}}
 {id=WavePlayer_90, properties {soundWaveAssetPtr=&ROCK_IMPACT_LAYER_HVY_002}}
]
properties {
 attenuationSettings=/Game/Sounds/Attenuation/Fort_Bullet_Impact_Att
 maxDistance=5300
 soundClassObject=&Fort_Impacts
 volumeMultiplier=1.416
 concurrencySet [&RifleBulletImpact_Conc]
}
