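# Sound cue for light bullet ricochet impacts on stone, using bullet-impact attenuation, the impacts sound class and rifle-impact concurrency.
# Each variant mixes a briefly, randomly delayed ricochet with an undelayed pre-impact whoosh, both pitch- and volume-modulated.
# A weighted empty branch leaves the whoosh out most of the time.
# A quality-level node gives its first slot larger wave pools; the other three share a trimmed mix.
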SoundCue /Game/Sounds/Fort_Weapons/Impacts/Bullet_Impacts_S01/FNS01_STONE_Bullet_Impact_Ricochet_Light_Cue
root=QualityLevel_0
nodes [
 {id=Delay_0, children [Modulator_0], properties {delayMax=0.15}}
 {id=Delay_1, children [Modulator_2], properties {delayMax=0.15}}
 {id=Mixer_0, children [Delay_0,Modulator_1], properties {inputVolume [1,1]}}
 {id=Mixer_1, children [Delay_1,Modulator_3], properties {inputVolume [1,1]}}
 {id=Modulator_0, children [Random_0], properties {pitchMax=1.2, pitchMin=0.9, volumeMax=1.5}}
 {id=Modulator_1, children [Random_1], properties {pitchMax=1.2, pitchMin=0.9, volumeMax=1.5}}
 {id=Modulator_2, children [Random_3], properties {pitchMax=1.2, pitchMin=0.9, volumeMax=1.5}}
 {id=Modulator_3, children [Random_4], properties {pitchMax=1.2, pitchMin=0.9, volumeMax=1.5}}
 {id=QualityLevel_0, children [Mixer_0,Mixer_1,Mixer_1,Mixer_1]}
 {
  id=Random_0
  children [
   WavePlayer_2
   WavePlayer_3
   WavePlayer_5
   WavePlayer_8
   WavePlayer_9
   WavePlayer_10
   WavePlayer_11
   WavePlayer_12
   WavePlayer_13
   WavePlayer_14
   WavePlayer_15
   WavePlayer_16
   WavePlayer_18
   WavePlayer_19
  ]
  properties {weights [1,1,1,1,1,1,1,1,1,1,1,1,1,1]}
 }
 {id=Random_1, children [Random_2,null], properties {bRandomizeWithoutReplacement=false, weights [0.3968,1]}}
 {id=Random_2, children [WavePlayer_20, WavePlayer_22, WavePlayer_26, WavePlayer_27, WavePlayer_28], properties {weights [1,1,1,1,1]}}
 {
  id=Random_3
  children [WavePlayer_29, WavePlayer_31, WavePlayer_32, WavePlayer_44, WavePlayer_45, WavePlayer_46]
  properties {weights [1,1,1,1,1,1]}
 }
 {id=Random_4, children [Random_5,null], properties {bRandomizeWithoutReplacement=false, weights [0.3968,1]}}
 {id=Random_5, children [WavePlayer_50, WavePlayer_47, WavePlayer_48, WavePlayer_49], properties {weights [1,1,1,1]}}
 {id=WavePlayer_10, properties {soundWaveAssetPtr=/Game/Sounds/Fort_Weapons/Impacts/Bullet_Impacts_S01/Wavs/ROCK_IMPACT_RICO_049}}
 {id=WavePlayer_11, properties {soundWaveAssetPtr=/Game/Sounds/Fort_Weapons/Impacts/Bullet_Impacts_S01/Wavs/ROCK_IMPACT_RICO_051}}
 {id=WavePlayer_12, properties {soundWaveAssetPtr=/Game/Sounds/Fort_Weapons/Impacts/Bullet_Impacts_S01/Wavs/ROCK_IMPACT_RICO_053}}
 {id=WavePlayer_13, properties {soundWaveAssetPtr=/Game/Sounds/Fort_Weapons/Impacts/Bullet_Impacts_S01/Wavs/ROCK_IMPACT_RICO_070}}
 {id=WavePlayer_14, properties {soundWaveAssetPtr=/Game/Sounds/Fort_Weapons/Impacts/Bullet_Impacts_S01/Wavs/ROCK_IMPACT_RICO_071}}
 {id=WavePlayer_15, properties {soundWaveAssetPtr=/Game/Sounds/Fort_Weapons/Impacts/Bullet_Impacts_S01/Wavs/ROCK_IMPACT_RICO_074}}
 {id=WavePlayer_16, properties {soundWaveAssetPtr=/Game/Sounds/Fort_Weapons/Impacts/Bullet_Impacts_S01/Wavs/ROCK_IMPACT_RICO_076}}
 {id=WavePlayer_18, properties {soundWaveAssetPtr=/Game/Sounds/Fort_Weapons/Impacts/Bullet_Impacts_S01/Wavs/ROCK_IMPACT_RICO_081}}
 {id=WavePlayer_19, properties {soundWaveAssetPtr=/Game/Sounds/Fort_Weapons/Impacts/Bullet_Impacts_S01/Wavs/ROCK_IMPACT_RICO_083}}
 {id=WavePlayer_2, properties {soundWaveAssetPtr=/Game/Sounds/Fort_Weapons/Impacts/Bullet_Impacts_S01/Wavs/ROCK_IMPACT_RICO_001}}
 {id=WavePlayer_20, properties {soundWaveAssetPtr=/Game/Sounds/Fort_Weapons/Impacts/Bullet_Impacts_S01/Wavs/ROCK_PRE_IMPACT_WHOOSH_000}}
 {id=WavePlayer_22, properties {soundWaveAssetPtr=/Game/Sounds/Fort_Weapons/Impacts/Bullet_Impacts_S01/Wavs/ROCK_PRE_IMPACT_WHOOSH_001}}
 {id=WavePlayer_26, properties {soundWaveAssetPtr=/Game/Sounds/Fort_Weapons/Impacts/Bullet_Impacts_S01/Wavs/ROCK_PRE_IMPACT_WHOOSH_002}}
 {id=WavePlayer_27, properties {soundWaveAssetPtr=/Game/Sounds/Fort_Weapons/Impacts/Bullet_Impacts_S01/Wavs/ROCK_PRE_IMPACT_WHOOSH_005}}
 {id=WavePlayer_28, properties {soundWaveAssetPtr=/Game/Sounds/Fort_Weapons/Impacts/Bullet_Impacts_S01/Wavs/ROCK_PRE_IMPACT_WHOOSH_006}}
 {id=WavePlayer_29, properties {soundWaveAssetPtr=&ROCK_IMPACT_RICO_001}}
 {id=WavePlayer_3, properties {soundWaveAssetPtr=/Game/Sounds/Fort_Weapons/Impacts/Bullet_Impacts_S01/Wavs/ROCK_IMPACT_RICO_037}}
 {id=WavePlayer_31, properties {soundWaveAssetPtr=&ROCK_IMPACT_RICO_037}}
 {id=WavePlayer_32, properties {soundWaveAssetPtr=/Game/Sounds/Fort_Weapons/Impacts/Bullet_Impacts_S01/Wavs/ROCK_IMPACT_RICO_043}}
 {id=WavePlayer_44, properties {soundWaveAssetPtr=/Game/Sounds/Fort_Weapons/Impacts/Bullet_Impacts_S01/Wavs/ROCK_IMPACT_RICO_047}}
 {id=WavePlayer_45, properties {soundWaveAssetPtr=&ROCK_IMPACT_RICO_071}}
 {id=WavePlayer_46, properties {soundWaveAssetPtr=&ROCK_IMPACT_RICO_083}}
 {id=WavePlayer_47, properties {soundWaveAssetPtr=&ROCK_PRE_IMPACT_WHOOSH_001}}
 {id=WavePlayer_48, properties {soundWaveAssetPtr=&ROCK_PRE_IMPACT_WHOOSH_005}}
 {id=WavePlayer_49, properties {soundWaveAssetPtr=&ROCK_PRE_IMPACT_WHOOSH_006}}
 {id=WavePlayer_5, properties {soundWaveAssetPtr=&ROCK_IMPACT_RICO_043}}
 {id=WavePlayer_50, properties {soundWaveAssetPtr=&ROCK_PRE_IMPACT_WHOOSH_000}}
 {id=WavePlayer_8, properties {soundWaveAssetPtr=/Game/Sounds/Fort_Weapons/Impacts/Bullet_Impacts_S01/Wavs/ROCK_IMPACT_RICO_045}}
 {id=WavePlayer_9, properties {soundWaveAssetPtr=&ROCK_IMPACT_RICO_047}}
]
properties {
 attenuationSettings=/Game/Sounds/Attenuation/Fort_Bullet_Impact_Att
 maxDistance=5300
 soundClassObject=&Fort_Impacts
 volumeMultiplier=1.416
 concurrencySet [&RifleBulletImpact_Conc]
}
