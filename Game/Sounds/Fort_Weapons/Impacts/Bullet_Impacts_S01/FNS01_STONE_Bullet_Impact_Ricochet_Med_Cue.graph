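# Sound cue for medium stone bullet ricochet impacts, with bullet-impact attenuation and rifle-impact concurrency.
# Each mix layers a slightly delayed random rock ricochet with an optional pre-impact whoosh, both pitch and volume modulated.
# A weighted empty random branch silences the whoosh more often than not.
# A quality switch gives the first level the larger wave pools, while the other three share a mix with smaller subsets.
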SoundCue /Game/Sounds/Fort_Weapons/Impacts/Bullet_Impacts_S01/FNS01_STONE_Bullet_Impact_Ricochet_Med_Cue
root=QualityLevel_2
nodes [
 {id=Delay_4, children [Modulator_8], properties {delayMax=0.15}}
 {id=Delay_5, children [Modulator_10], properties {delayMax=0.15}}
 {id=Mixer_4, children [Delay_4,Modulator_9], properties {inputVolume [1,1]}}
 {id=Mixer_5, children [Delay_5,Modulator_11], properties {inputVolume [1,1]}}
 {id=Modulator_10, children [Random_15], properties {pitchMax=1.1, pitchMin=0.9, volumeMax=2, volumeMin=1}}
 {id=Modulator_11, children [Random_16], properties {pitchMax=1.25, pitchMin=0.9, volumeMax=1.5}}
 {id=Modulator_8, children [Random_12], properties {pitchMax=1.1, pitchMin=0.9, volumeMax=2, volumeMin=1}}
 {id=Modulator_9, children [Random_13], properties {pitchMax=1.2, pitchMin=0.9, volumeMax=1.5}}
 {id=QualityLevel_2, children [Mixer_4,Mixer_5,Mixer_5,Mixer_5]}
 {
  id=Random_12
  children [
   WavePlayer_11
   WavePlayer_12
   WavePlayer_13
   WavePlayer_14
   WavePlayer_16
   WavePlayer_18
   WavePlayer_20
   WavePlayer_30
   WavePlayer_31
   WavePlayer_32
   WavePlayer_33
   WavePlayer_34
   WavePlayer_35
   WavePlayer_36
   WavePlayer_37
   WavePlayer_38
   WavePlayer_39
   WavePlayer_40
   WavePlayer_41
   WavePlayer_42
   WavePlayer_43
   WavePlayer_44
   WavePlayer_45
  ]
  properties {weights [1,1,1,1,1,1,1,1,1,1,1,1,1,1,1,1,1,1,1,1,1,1,1]}
 }
 {id=Random_13, children [Random_14,null], properties {bRandomizeWithoutReplacement=false, weights [0.7087,1]}}
 {
  id=Random_14
  children [WavePlayer_46, WavePlayer_47, WavePlayer_48, WavePlayer_49, WavePlayer_50, WavePlayer_51]
  properties {weights [1,1,1,1,1,1]}
 }
 {
  id=Random_15
  children [
   WavePlayer_52
   WavePlayer_53
   WavePlayer_54
   WavePlayer_55
   WavePlayer_58
   WavePlayer_59
   WavePlayer_60
   WavePlayer_61
   WavePlayer_62
   WavePlayer_63
   WavePlayer_64
  ]
  properties {weights [1,1,1,1,1,1,1,1,1,1,1]}
 }
 {id=Random_16, children [Random_17,null], properties {bRandomizeWithoutReplacement=false, weights [0.7087,1]}}
 {id=Random_17, children [WavePlayer_66, WavePlayer_67, WavePlayer_68], properties {weights [1,1,1]}}
 {id=WavePlayer_11, properties {soundWaveAssetPtr=/Game/Sounds/Fort_Weapons/Impacts/Bullet_Impacts_S01/Wavs/ROCK_IMPACT_RICO_054}}
 {id=WavePlayer_12, properties {soundWaveAssetPtr=/Game/Sounds/Fort_Weapons/Impacts/Bullet_Impacts_S01/Wavs/ROCK_IMPACT_RICO_055}}
 {id=WavePlayer_13, properties {soundWaveAssetPtr=/Game/Sounds/Fort_Weapons/Impacts/Bullet_Impacts_S01/Wavs/ROCK_IMPACT_RICO_056}}
 {id=WavePlayer_14, properties {soundWaveAssetPtr=/Game/Sounds/Fort_Weapons/Impacts/Bullet_Impacts_S01/Wavs/ROCK_IMPACT_RICO_058}}
 {id=WavePlayer_16, properties {soundWaveAssetPtr=/Game/Sounds/Fort_Weapons/Impacts/Bullet_Impacts_S01/Wavs/ROCK_IMPACT_RICO_060}}
 {id=WavePlayer_18, properties {soundWaveAssetPtr=/Game/Sounds/Fort_Weapons/Impacts/Bullet_Impacts_S01/Wavs/ROCK_IMPACT_RICO_061}}
 {id=WavePlayer_20, properties {soundWaveAssetPtr=/Game/Sounds/Fort_Weapons/Impacts/Bullet_Impacts_S01/Wavs/ROCK_IMPACT_RICO_062}}
 {id=WavePlayer_30, properties {soundWaveAssetPtr=/Game/Sounds/Fort_Weapons/Impacts/Bullet_Impacts_S01/Wavs/ROCK_IMPACT_RICO_064}}
 {id=WavePlayer_31, properties {soundWaveAssetPtr=/Game/Sounds/Fort_Weapons/Impacts/Bullet_Impacts_S01/Wavs/ROCK_IMPACT_RICO_065}}
 {id=WavePlayer_32, properties {soundWaveAssetPtr=/Game/Sounds/Fort_Weapons/Impacts/Bullet_Impacts_S01/Wavs/ROCK_IMPACT_RICO_068}}
 {id=WavePlayer_33, properties {soundWaveAssetPtr=/Game/Sounds/Fort_Weapons/Impacts/Bullet_Impacts_S01/Wavs/ROCK_IMPACT_RICO_069}}
 {id=WavePlayer_34, properties {soundWaveAssetPtr=/Game/Sounds/Fort_Weapons/Impacts/Bullet_Impacts_S01/Wavs/ROCK_IMPACT_RICO_072}}
 {id=WavePlayer_35, properties {soundWaveAssetPtr=/Game/Sounds/Fort_Weapons/Impacts/Bullet_Impacts_S01/Wavs/ROCK_IMPACT_RICO_073}}
 {id=WavePlayer_36, properties {soundWaveAssetPtr=/Game/Sounds/Fort_Weapons/Impacts/Bullet_Impacts_S01/Wavs/ROCK_IMPACT_RICO_074}}
 {id=WavePlayer_37, properties {soundWaveAssetPtr=/Game/Sounds/Fort_Weapons/Impacts/Bullet_Impacts_S01/Wavs/ROCK_IMPACT_RICO_075}}
 {id=WavePlayer_38, properties {soundWaveAssetPtr=/Game/Sounds/Fort_Weapons/Impacts/Bullet_Impacts_S01/Wavs/ROCK_IMPACT_RICO_076}}
 {id=WavePlayer_39, properties {soundWaveAssetPtr=/Game/Sounds/Fort_Weapons/Impacts/Bullet_Impacts_S01/Wavs/ROCK_IMPACT_RICO_077}}
 {id=WavePlayer_40, properties {soundWaveAssetPtr=/Game/Sounds/Fort_Weapons/Impacts/Bullet_Impacts_S01/Wavs/ROCK_IMPACT_RICO_078}}
 {id=WavePlayer_41, properties {soundWaveAssetPtr=/Game/Sounds/Fort_Weapons/Impacts/Bullet_Impacts_S01/Wavs/ROCK_IMPACT_RICO_079}}
 {id=WavePlayer_42, properties {soundWaveAssetPtr=/Game/Sounds/Fort_Weapons/Impacts/Bullet_Impacts_S01/Wavs/ROCK_IMPACT_RICO_082}}
 {id=WavePlayer_43, properties {soundWaveAssetPtr=/Game/Sounds/Fort_Weapons/Impacts/Bullet_Impacts_S01/Wavs/ROCK_IMPACT_RICO_084}}
 {id=WavePlayer_44, properties {soundWaveAssetPtr=/Game/Sounds/Fort_Weapons/Impacts/Bullet_Impacts_S01/Wavs/ROCK_IMPACT_RICO_085}}
 {id=WavePlayer_45, properties {soundWaveAssetPtr=/Game/Sounds/Fort_Weapons/Impacts/Bullet_Impacts_S01/Wavs/ROCK_IMPACT_RICO_086}}
 {id=WavePlayer_46, properties {soundWaveAssetPtr=/Game/Sounds/Fort_Weapons/Impacts/Bullet_Impacts_S01/Wavs/ROCK_PRE_IMPACT_WHOOSH_000}}
 {id=WavePlayer_47, properties {soundWaveAssetPtr=/Game/Sounds/Fort_Weapons/Impacts/Bullet_Impacts_S01/Wavs/ROCK_PRE_IMPACT_WHOOSH_001}}
 {id=WavePlayer_48, properties {soundWaveAssetPtr=/Game/Sounds/Fort_Weapons/Impacts/Bullet_Impacts_S01/Wavs/ROCK_PRE_IMPACT_WHOOSH_002}}
 {id=WavePlayer_49, properties {soundWaveAssetPtr=/Game/Sounds/Fort_Weapons/Impacts/Bullet_Impacts_S01/Wavs/ROCK_PRE_IMPACT_WHOOSH_003}}
 {id=WavePlayer_50, properties {soundWaveAssetPtr=/Game/Sounds/Fort_Weapons/Impacts/Bullet_Impacts_S01/Wavs/ROCK_PRE_IMPACT_WHOOSH_005}}
 {id=WavePlayer_51, properties {soundWaveAssetPtr=/Game/Sounds/Fort_Weapons/Impacts/Bullet_Impacts_S01/Wavs/ROCK_PRE_IMPACT_WHOOSH_007}}
 {id=WavePlayer_52, properties {soundWaveAssetPtr=&ROCK_IMPACT_RICO_054}}
 {id=WavePlayer_53, properties {soundWaveAssetPtr=&ROCK_IMPACT_RICO_058}}
 {id=WavePlayer_54, properties {soundWaveAssetPtr=&ROCK_IMPACT_RICO_061}}
 {id=WavePlayer_55, properties {soundWaveAssetPtr=&ROCK_IMPACT_RICO_062}}
 {id=WavePlayer_58, properties {soundWaveAssetPtr=&ROCK_IMPACT_RICO_074}}
 {id=WavePlayer_59, properties {soundWaveAssetPtr=&ROCK_IMPACT_RICO_077}}
 {id=WavePlayer_60, properties {soundWaveAssetPtr=&ROCK_IMPACT_RICO_078}}
 {id=WavePlayer_61, properties {soundWaveAssetPtr=&ROCK_IMPACT_RICO_079}}
 {id=WavePlayer_62, properties {soundWaveAssetPtr=&ROCK_IMPACT_RICO_082}}
 {id=WavePlayer_63, properties {soundWaveAssetPtr=&ROCK_IMPACT_RICO_084}}
 {id=WavePlayer_64, properties {soundWaveAssetPtr=&ROCK_IMPACT_RICO_085}}
 {id=WavePlayer_66, properties {soundWaveAssetPtr=&ROCK_PRE_IMPACT_WHOOSH_000}}
 {id=WavePlayer_67, properties {soundWaveAssetPtr=&ROCK_PRE_IMPACT_WHOOSH_001}}
 {id=WavePlayer_68, properties {soundWaveAssetPtr=&ROCK_PRE_IMPACT_WHOOSH_002}}
]
properties {
 attenuationSettings=/Game/Sounds/Attenuation/Fort_Bullet_Impact_Att
 maxDistance=5300
 soundClassObject=&Fort_Impacts
 volumeMultiplier=1.416
 concurrencySet [&RifleBulletImpact_Conc]
}
