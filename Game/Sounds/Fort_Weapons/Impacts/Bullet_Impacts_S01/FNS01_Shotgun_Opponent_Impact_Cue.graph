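# Sound cue for shotgun bullet impacts on an opponent, routed to the Fort_Impacts sound class with bullet-impact attenuation.
# A quality-level node sends its first level to a random pick of eight impact waves and the other three levels to a shared five-wave subset.
# Each branch passes through a modulator that varies pitch and volume per play.
# The cue references the rifle bullet impact concurrency set to limit overlapping impacts.
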SoundCue /Game/Sounds/Fort_Weapons/Impacts/Bullet_Impacts_S01/FNS01_Shotgun_Opponent_Impact_Cue
root=QualityLevel_1
nodes [
 {id=Modulator_0, children [Random_0], properties {pitchMax=1.2, pitchMin=0.9, volumeMax=1.5}}
 {id=Modulator_3, children [Random_3], properties {pitchMax=1.2, pitchMin=0.9, volumeMax=1.5}}
 {id=QualityLevel_1, children [Modulator_0,Modulator_3,Modulator_3,Modulator_3]}
 {
  id=Random_0
  children [WavePlayer_0, WavePlayer_1, WavePlayer_2, WavePlayer_3, WavePlayer_17, WavePlayer_18, WavePlayer_19, WavePlayer_20]
  properties {weights [1,1,1,1,1,1,1,1]}
 }
 {id=Random_3, children [WavePlayer_21, WavePlayer_22, WavePlayer_23, WavePlayer_24, WavePlayer_25], properties {weights [1,1,1,1,1]}}
 {id=WavePlayer_0, properties {soundWaveAssetPtr=/Game/Sounds/Fort_Weapons/Impacts/opponent_bullet_impact_01}}
 {id=WavePlayer_1, properties {soundWaveAssetPtr=/Game/Sounds/Fort_Weapons/Impacts/opponent_bullet_impact_02}}
 {id=WavePlayer_17, properties {soundWaveAssetPtr=/Game/Sounds/Fort_Weapons/Impacts/opponent_bullet_impact_05}}
 {id=WavePlayer_18, properties {soundWaveAssetPtr=/Game/Sounds/Fort_Weapons/Impacts/opponent_bullet_impact_06}}
 {id=WavePlayer_19, properties {soundWaveAssetPtr=/Game/Sounds/Fort_Weapons/Impacts/opponent_bullet_impact_07}}
 {id=WavePlayer_2, properties {soundWaveAssetPtr=/Game/Sounds/Fort_Weapons/Impacts/opponent_bullet_impact_03}}
 {id=WavePlayer_20, properties {soundWaveAssetPtr=/Game/Sounds/Fort_Weapons/Impacts/opponent_bullet_impact_08}}
 {id=WavePlayer_21, properties {soundWaveAssetPtr=&opponent_bullet_impact_01}}
 {id=WavePlayer_22, properties {soundWaveAssetPtr=&opponent_bullet_impact_02}}
 {id=WavePlayer_23, properties {soundWaveAssetPtr=&opponent_bullet_impact_03}}
 {id=WavePlayer_24, properties {soundWaveAssetPtr=&opponent_bullet_impact_07}}
 {id=WavePlayer_25, properties {soundWaveAssetPtr=&opponent_bullet_impact_08}}
 {id=WavePlayer_3, properties {soundWaveAssetPtr=/Game/Sounds/Fort_Weapons/Impacts/opponent_bullet_impact_04}}
]
properties {
 attenuationSettings=/Game/Sounds/Attenuation/Fort_Bullet_Impact_Att
 maxDistance=5300
 soundClassObject=&Fort_Impacts
 volumeMultiplier=1.416
 concurrencyOverrides {maxCount=4, resolutionRule=PreventNew}
 concurrencySet [&RifleBulletImpact_Conc]
}
