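# Sound cue for generic shotgun ricochet impacts, played in the impacts sound class with bullet-impact attenuation and rifle-impact concurrency.
# A quality switch gives the first level five source waves and the other levels a shared four-wave branch.
# Each branch mixes two random picks, one pitched down, from plain and pitch-enveloped ricochets.
# The mix passes through volume modulation, an attenuation node and a short random delay.
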SoundCue /Game/Sounds/Fort_Weapons/Impacts/Bullet_Impacts_S01/FNS01_Shotgun_Ricochet_Generic_Cue
root=QualityLevel_0
nodes [
 {id=Attenuation_0, children [Modulator_2], properties {bOverrideAttenuation=true}}
 {id=Attenuation_1, children [Modulator_7], properties {bOverrideAttenuation=true}}
 {id=Delay_0, children [Attenuation_0], properties {delayMax=0.15, delayMin=0.05}}
 {id=Delay_1, children [Attenuation_1], properties {delayMax=0.15, delayMin=0.05}}
 {
  id=Enveloper_0
  children [WavePlayer_1]
  properties {pitchCurve {editorCurveData {keys [{time=0}, <at 0.2252 1.4109>, <at 0.456 -0.1395>]}}, volumeCurve {editorCurveData {keys [<at 0 1>]}}}
 }
 {
  id=Enveloper_1
  children [WavePlayer_2]
  properties {
   pitchCurve {editorCurveData {keys [{time=0}, <at 0.0897 1.4972>, <at 0.254 0.6319>, <at 0.7489 0.4609>]}}
   volumeCurve {editorCurveData {keys [<at 0 1>]}}
  }
 }
 {
  id=Enveloper_2
  children [WavePlayer_14]
  properties {pitchCurve {editorCurveData {keys [{time=0}, <at 0.0771 1.4823>, <at 0.456 -0.1395>]}}, volumeCurve {editorCurveData {keys [<at 0 1>]}}}
 }
 {
  id=Enveloper_3
  children [WavePlayer_16]
  properties {
   pitchCurve {editorCurveData {keys [{time=0}, <at 0.0897 1.4972>, <at 0.254 0.6319>, <at 0.7489 0.4609>]}}
   volumeCurve {editorCurveData {keys [<at 0 1>]}}
  }
 }
 {
  id=Enveloper_4
  children [WavePlayer_18]
  properties {pitchCurve {editorCurveData {keys [{time=0}, <at 0.0771 1.4823>, <at 0.456 -0.1395>]}}, volumeCurve {editorCurveData {keys [<at 0 1>]}}}
 }
 {id=Mixer_0, children [Modulator_0,Modulator_1], properties {inputVolume [1,1]}}
 {id=Mixer_1, children [Modulator_3,Modulator_6], properties {inputVolume [1,1]}}
 {id=Modulator_0, children [Random_0]}
 {id=Modulator_1, children [Random_1], properties {pitchMax=0.7, pitchMin=0.5}}
 {id=Modulator_2, children [Mixer_0], properties {volumeMax=1.5}}
 {id=Modulator_3, children [Random_2]}
 {id=Modulator_6, children [Random_3], properties {pitchMax=0.7, pitchMin=0.5}}
 {id=Modulator_7, children [Mixer_1], properties {volumeMax=1.5}}
 {id=QualityLevel_0, children [Delay_0,Delay_1,Delay_1,Delay_1]}
 {
  id=Random_0
  children [WavePlayer_0, WavePlayer_1, WavePlayer_2, WavePlayer_3, WavePlayer_14, Enveloper_0, Enveloper_1, Enveloper_2]
  properties {weights [1,1,1,1,1,1,1,1]}
 }
 {
  id=Random_1
  children [WavePlayer_0, WavePlayer_1, WavePlayer_2, WavePlayer_3, WavePlayer_14, Enveloper_0, Enveloper_1, Enveloper_2]
  properties {weights [1,1,1,1,1,1,1,1]}
 }
 {
  id=Random_2
  children [WavePlayer_15, WavePlayer_16, WavePlayer_17, WavePlayer_18, Enveloper_3, Enveloper_4]
  properties {weights [1,1,1,1,1,1]}
 }
 {
  id=Random_3
  children [WavePlayer_15, WavePlayer_16, WavePlayer_17, WavePlayer_18, Enveloper_3, Enveloper_4]
  properties {weights [1,1,1,1,1,1]}
 }
 {id=WavePlayer_0, properties {soundWaveAssetPtr=/Game/Sounds/Fort_Weapons/Impacts/Bullet_Impacts_S01/Wavs/fort_ricochet_01}}
 {id=WavePlayer_1, properties {soundWaveAssetPtr=/Game/Sounds/Fort_Weapons/Impacts/Bullet_Impacts_S01/Wavs/fort_ricochet_02}}
 {id=WavePlayer_14, properties {soundWaveAssetPtr=/Game/Sounds/Fort_Weapons/Impacts/Bullet_Impacts_S01/Wavs/fort_ricochet_05}}
 {id=WavePlayer_15, properties {soundWaveAssetPtr=&fort_ricochet_01}}
 {id=WavePlayer_16, properties {soundWaveAssetPtr=/Game/Sounds/Fort_Weapons/Impacts/Bullet_Impacts_S01/Wavs/fort_ricochet_03}}
 {id=WavePlayer_17, properties {soundWaveAssetPtr=/Game/Sounds/Fort_Weapons/Impacts/Bullet_Impacts_S01/Wavs/fort_ricochet_04}}
 {id=WavePlayer_18, properties {soundWaveAssetPtr=&fort_ricochet_05}}
 {id=WavePlayer_2, properties {soundWaveAssetPtr=&fort_ricochet_03}}
 {id=WavePlayer_3, properties {soundWaveAssetPtr=&fort_ricochet_04}}
]
properties {
 attenuationSettings=/Game/Sounds/Attenuation/Fort_Bullet_Impact_Att
 maxDistance=5300
 soundClassObject=&Fort_Impacts
 volumeMultiplier=1.416
 concurrencyOverrides {maxCount=4, resolutionRule=PreventNew}
 concurrencySet [&RifleBulletImpact_Conc]
}
