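# Heavy bullet impact and ricochet cue for wood surfaces, using bullet impact attenuation, rifle impact concurrency and the impacts sound class.
# Each mix layers a randomly delayed main hit with three chance-based layers (second hit, delayed extra layer, WSHS impact), each randomized in pitch and volume.
# Random nodes with empty branches let those extra layers drop out.
# A quality switch gives the first level nine main-hit variations and the other three levels a shared six-variation mix.
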SoundCue /Game/Sounds/Fort_Weapons/Impacts/Bullet_Impacts_S01/FNS01_WOOD_Bullet_Impact_Ricochet_Heavy_Cue
root=QualityLevel_0
nodes [
 {id=Delay_0, children [Modulator_0], properties {delayMax=0.2}}
 {id=Delay_1, children [Modulator_3], properties {delayMax=0.2}}
 {id=Delay_2, children [Modulator_5], properties {delayMax=0.2}}
 {id=Delay_4, children [Modulator_7], properties {delayMax=0.2}}
 {id=Mixer_0, children [Delay_0,Modulator_2,Delay_1,Modulator_4], properties {inputVolume [1,1,1,1]}}
 {id=Mixer_1, children [Delay_2,Modulator_6,Delay_4,Modulator_8], properties {inputVolume [1,1,1,1]}}
 {id=Modulator_0, children [Random_0], properties {pitchMax=1.1, pitchMin=0.9, volumeMax=1.5}}
 {id=Modulator_2, children [Random_2], properties {pitchMax=1.1, pitchMin=0.8, volumeMax=1.5}}
 {id=Modulator_3, children [Random_3], properties {pitchMax=1.1, pitchMin=0.8, volumeMax=1.5, volumeMin=0.85}}
 {id=Modulator_4, children [Random_4], properties {pitchMax=1.1, volumeMax=1.1, volumeMin=0.85}}
 {id=Modulator_5, children [Random_6], properties {pitchMax=1.1, pitchMin=0.9, volumeMax=1.5}}
 {id=Modulator_6, children [Random_8], properties {pitchMax=1.1, pitchMin=0.8, volumeMax=1.5}}
 {id=Modulator_7, children [Random_9], properties {pitchMax=1.1, pitchMin=0.8, volumeMax=1.5, volumeMin=0.85}}
 {id=Modulator_8, children [Random_10], properties {pitchMax=1.1, volumeMax=1.1, volumeMin=0.85}}
 {id=QualityLevel_0, children [Mixer_0,Mixer_1,Mixer_1,Mixer_1]}
 {
  id=Random_0
  children [WavePlayer_31, WavePlayer_32, WavePlayer_34, WavePlayer_36, WavePlayer_37, WavePlayer_39, WavePlayer_43, WavePlayer_49, WavePlayer_50]
  properties {weights [1,1,1,1,1,1,1,1,1]}
 }
 {id=Random_1, children [WavePlayer_52, WavePlayer_53, WavePlayer_54], properties {weights [1,1,1]}}
 {id=Random_10, children [Random_11,null], properties {bRandomizeWithoutReplacement=false, weights [0.4495,1]}}
 {id=Random_11, children [WavePlayer_70, WavePlayer_71, WavePlayer_72, WavePlayer_73], properties {weights [1,1,1,1]}}
 {id=Random_2, children [Random_1,null], properties {bRandomizeWithoutReplacement=false, weights [0.9449,1]}}
 {id=Random_3, children [WavePlayer_55, null], properties {bRandomizeWithoutReplacement=false, weights [0.5045,1]}}
 {id=Random_4, children [Random_5,null], properties {bRandomizeWithoutReplacement=false, weights [0.4495,1]}}
 {id=Random_5, children [WavePlayer_56, WavePlayer_57, WavePlayer_58, WavePlayer_59], properties {weights [1,1,1,1]}}
 {
  id=Random_6
  children [WavePlayer_60, WavePlayer_61, WavePlayer_62, WavePlayer_63, WavePlayer_64, WavePlayer_65]
  properties {weights [1,1,1,1,1,1]}
 }
 {id=Random_7, children [WavePlayer_66, WavePlayer_67, WavePlayer_68], properties {weights [1,1,1]}}
 {id=Random_8, children [Random_7,null], properties {bRandomizeWithoutReplacement=false, weights [0.9449,1]}}
 {id=Random_9, children [WavePlayer_69, null], properties {bRandomizeWithoutReplacement=false, weights [0.5045,1]}}
 {id=WavePlayer_31, properties {soundWaveAssetPtr=/Game/Sounds/Fort_Weapons/Impacts/Bullet_Impacts_S01/Wavs/WOOD_IMPACT_HVY_000}}
 {id=WavePlayer_32, properties {soundWaveAssetPtr=/Game/Sounds/Fort_Weapons/Impacts/Bullet_Impacts_S01/Wavs/WOOD_IMPACT_HVY_001}}
 {id=WavePlayer_34, properties {soundWaveAssetPtr=/Game/Sounds/Fort_Weapons/Impacts/Bullet_Impacts_S01/Wavs/WOOD_IMPACT_HVY_003}}
 {id=WavePlayer_36, properties {soundWaveAssetPtr=/Game/Sounds/Fort_Weapons/Impacts/Bullet_Impacts_S01/Wavs/WOOD_IMPACT_HVY_004}}
 {id=WavePlayer_37, properties {soundWaveAssetPtr=/Game/Sounds/Fort_Weapons/Impacts/Bullet_Impacts_S01/Wavs/WOOD_IMPACT_HVY_005}}
 {id=WavePlayer_39, properties {soundWaveAssetPtr=/Game/Sounds/Fort_Weapons/Impacts/Bullet_Impacts_S01/Wavs/WOOD_IMPACT_HVY_007}}
 {id=WavePlayer_43, properties {soundWaveAssetPtr=/Game/Sounds/Fort_Weapons/Impacts/Bullet_Impacts_S01/Wavs/WOOD_IMPACT_HVY_010}}
 {id=WavePlayer_49, properties {soundWaveAssetPtr=/Game/Sounds/Fort_Weapons/Impacts/Bullet_Impacts_S01/Wavs/WOOD_IMPACT_HVY_011}}
 {id=WavePlayer_50, properties {soundWaveAssetPtr=/Game/Sounds/Fort_Weapons/Impacts/Bullet_Impacts_S01/Wavs/WOOD_IMPACT_HVY_012}}
 {id=WavePlayer_52, properties {soundWaveAssetPtr=&WOOD_IMPACT_HVY_003}}
 {id=WavePlayer_53, properties {soundWaveAssetPtr=&WOOD_IMPACT_HVY_007}}
 {id=WavePlayer_54, properties {soundWaveAssetPtr=&WOOD_IMPACT_HVY_011}}
 {id=WavePlayer_55, properties {soundWaveAssetPtr=/Game/Sounds/Fort_Weapons/Impacts/Bullet_Impacts_S01/Wavs/WOOD_IMPACT_LAYER_HVY_000}}
 {id=WavePlayer_56, properties {soundWaveAssetPtr=/Game/Sounds/Fort_Weapons/Impacts/Bullet_Impacts_S01/Wavs/WOOD_WSHS_IMPACT_HVY_000}}
 {id=WavePlayer_57, properties {soundWaveAssetPtr=/Game/Sounds/Fort_Weapons/Impacts/Bullet_Impacts_S01/Wavs/WOOD_WSHS_IMPACT_HVY_001}}
 {id=WavePlayer_58, properties {soundWaveAssetPtr=/Game/Sounds/Fort_Weapons/Impacts/Bullet_Impacts_S01/Wavs/WOOD_WSHS_IMPACT_HVY_002}}
 {id=WavePlayer_59, properties {soundWaveAssetPtr=/Game/Sounds/Fort_Weapons/Impacts/Bullet_Impacts_S01/Wavs/WOOD_WSHS_IMPACT_HVY_003}}
 {id=WavePlayer_60, properties {soundWaveAssetPtr=&WOOD_IMPACT_HVY_000}}
 {id=WavePlayer_61, properties {soundWaveAssetPtr=&WOOD_IMPACT_HVY_001}}
 {id=WavePlayer_62, properties {soundWaveAssetPtr=&WOOD_IMPACT_HVY_003}}
 {id=WavePlayer_63, properties {soundWaveAssetPtr=&WOOD_IMPACT_HVY_005}}
 {id=WavePlayer_64, properties {soundWaveAssetPtr=&WOOD_IMPACT_HVY_010}}
 {id=WavePlayer_65, properties {soundWaveAssetPtr=&WOOD_IMPACT_HVY_012}}
 {id=WavePlayer_66, properties {soundWaveAssetPtr=&WOOD_IMPACT_HVY_003}}
 {id=WavePlayer_67, properties {soundWaveAssetPtr=&WOOD_IMPACT_HVY_007}}
 {id=WavePlayer_68, properties {soundWaveAssetPtr=&WOOD_IMPACT_HVY_011}}
 {id=WavePlayer_69, properties {soundWaveAssetPtr=&WOOD_IMPACT_LAYER_HVY_000}}
 {id=WavePlayer_70, properties {soundWaveAssetPtr=&WOOD_WSHS_IMPACT_HVY_000}}
 {id=WavePlayer_71, properties {soundWaveAssetPtr=&WOOD_WSHS_IMPACT_HVY_001}}
 {id=WavePlayer_72, properties {soundWaveAssetPtr=&WOOD_WSHS_IMPACT_HVY_002}}
 {id=WavePlayer_73, properties {soundWaveAssetPtr=&WOOD_WSHS_IMPACT_HVY_003}}
]
properties {
 attenuationSettings=/Game/Sounds/Attenuation/Fort_Bullet_Impact_Att
 maxDistance=5300
 soundClassObject=&Fort_Impacts
 volumeMultiplier=1.416
 concurrencySet [&RifleBulletImpact_Conc]
}
