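# Sound cue for light bullet impacts on wood, named as a ricochet variant, using shared bullet-impact attenuation and rifle impact concurrency.
# Each playback mixes a random wood impact with a random wood break, each layer modulated for pitch and volume.
# A quality-level root gives the first level the larger wave pools.
# The other three levels share one mixer with smaller subsets of the same waves.
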SoundCue /Game/Sounds/Fort_Weapons/Impacts/Bullet_Impacts_S01/FNS01_WOOD_Bullet_Impact_Ricochet_Light_Cue
root=QualityLevel_4
nodes [
 {id=Mixer_12, children [Modulator_24,Modulator_25], properties {inputVolume [1,1]}}
 {id=Mixer_13, children [Modulator_26,Modulator_27], properties {inputVolume [1,1]}}
 {id=Modulator_24, children [Random_25], properties {pitchMax=1.2, pitchMin=0.9, volumeMax=1.5}}
 {id=Modulator_25, children [Random_24], properties {pitchMax=1.2, pitchMin=0.9, volumeMax=2}}
 {id=Modulator_26, children [Random_27], properties {pitchMax=1.2, pitchMin=0.9, volumeMax=1.5}}
 {id=Modulator_27, children [Random_26], properties {pitchMax=1.2, pitchMin=0.9, volumeMax=2}}
 {id=QualityLevel_4, children [Mixer_12,Mixer_13,Mixer_13,Mixer_13]}
 {
  id=Random_24
  children [WavePlayer_21, WavePlayer_23, WavePlayer_24, WavePlayer_25, WavePlayer_30, WavePlayer_33, WavePlayer_34]
  properties {weights [1,1,1,1,1,1,1]}
 }
 {
  id=Random_25
  children [WavePlayer_0, WavePlayer_1, WavePlayer_4, WavePlayer_6, WavePlayer_7, WavePlayer_17]
  properties {weights [1,1,1,1,1,1]}
 }
 {id=Random_26, children [WavePlayer_42, WavePlayer_41, WavePlayer_40, WavePlayer_39, WavePlayer_43], properties {weights [1,1,1,1,1]}}
 {id=Random_27, children [WavePlayer_35, WavePlayer_36, WavePlayer_37, WavePlayer_38], properties {weights [1,1,1,1]}}
 {id=WavePlayer_0, properties {soundWaveAssetPtr=/Game/Sounds/Fort_Weapons/Impacts/Bullet_Impacts_S01/Wavs/WOOD_IMPACT_LIGHT_000}}
 {id=WavePlayer_1, properties {soundWaveAssetPtr=/Game/Sounds/Fort_Weapons/Impacts/Bullet_Impacts_S01/Wavs/WOOD_IMPACT_LIGHT_001}}
 {id=WavePlayer_17, properties {soundWaveAssetPtr=/Game/Sounds/Fort_Weapons/Impacts/Bullet_Impacts_S01/Wavs/WOOD_IMPACT_LIGHT_009}}
 {id=WavePlayer_21, properties {soundWaveAssetPtr=/Game/Sounds/Fort_Weapons/Impacts/Bullet_Impacts_S01/Wavs/WOOD_IMPACT_BREAKS_LIGHT_003}}
 {id=WavePlayer_23, properties {soundWaveAssetPtr=/Game/Sounds/Fort_Weapons/Impacts/Bullet_Impacts_S01/Wavs/WOOD_IMPACT_BREAKS_LIGHT_005}}
 {id=WavePlayer_24, properties {soundWaveAssetPtr=/Game/Sounds/Fort_Weapons/Impacts/Bullet_Impacts_S01/Wavs/WOOD_IMPACT_BREAKS_LIGHT_006}}
 {id=WavePlayer_25, properties {soundWaveAssetPtr=/Game/Sounds/Fort_Weapons/Impacts/Bullet_Impacts_S01/Wavs/WOOD_IMPACT_BREAKS_LIGHT_007}}
 {id=WavePlayer_30, properties {soundWaveAssetPtr=/Game/Sounds/Fort_Weapons/Impacts/Bullet_Impacts_S01/Wavs/WOOD_IMPACT_BREAKS_LIGHT_012}}
 {id=WavePlayer_33, properties {soundWaveAssetPtr=/Game/Sounds/Fort_Weapons/Impacts/Bullet_Impacts_S01/Wavs/WOOD_IMPACT_BREAKS_LIGHT_015}}
 {id=WavePlayer_34, properties {soundWaveAssetPtr=/Game/Sounds/Fort_Weapons/Impacts/Bullet_Impacts_S01/Wavs/WOOD_IMPACT_BREAKS_LIGHT_016}}
 {id=WavePlayer_35, properties {soundWaveAssetPtr=&WOOD_IMPACT_LIGHT_000}}
 {id=WavePlayer_36, properties {soundWaveAssetPtr=/Game/Sounds/Fort_Weapons/Impacts/Bullet_Impacts_S01/Wavs/WOOD_IMPACT_LIGHT_004}}
 {id=WavePlayer_37, properties {soundWaveAssetPtr=/Game/Sounds/Fort_Weapons/Impacts/Bullet_Impacts_S01/Wavs/WOOD_IMPACT_LIGHT_007}}
 {id=WavePlayer_38, properties {soundWaveAssetPtr=&WOOD_IMPACT_LIGHT_009}}
 {id=WavePlayer_39, properties {soundWaveAssetPtr=&WOOD_IMPACT_BREAKS_LIGHT_007}}
 {id=WavePlayer_4, properties {soundWaveAssetPtr=&WOOD_IMPACT_LIGHT_004}}
 {id=WavePlayer_40, properties {soundWaveAssetPtr=&WOOD_IMPACT_BREAKS_LIGHT_006}}
 {id=WavePlayer_41, properties {soundWaveAssetPtr=&WOOD_IMPACT_BREAKS_LIGHT_005}}
 {id=WavePlayer_42, properties {soundWaveAssetPtr=&WOOD_IMPACT_BREAKS_LIGHT_003}}
 {id=WavePlayer_43, properties {soundWaveAssetPtr=&WOOD_IMPACT_BREAKS_LIGHT_012}}
 {id=WavePlayer_6, properties {soundWaveAssetPtr=/Game/Sounds/Fort_Weapons/Impacts/Bullet_Impacts_S01/Wavs/WOOD_IMPACT_LIGHT_006}}
 {id=WavePlayer_7, properties {soundWaveAssetPtr=&WOOD_IMPACT_LIGHT_007}}
]
properties {
 attenuationSettings=/Game/Sounds/Attenuation/Fort_Bullet_Impact_Att
 maxDistance=5300
 soundClassObject=&Fort_Impacts
 volumeMultiplier=1.416
 concurrencySet [&RifleBulletImpact_Conc]
}
