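# Sound cue for medium bullet impacts on wood, named as a ricochet variant, playing randomized WOOD_IMPACT_MED waves.
# A Quality Level root sends the first level to a 14-wave random pool, while the other three share one branch with a nine-wave subset.
# Each pool passes through a Modulator that varies pitch and volume.
# Bullet impact attenuation, the impacts sound class and rifle impact concurrency govern playback.
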SoundCue /Game/Sounds/Fort_Weapons/Impacts/Bullet_Impacts_S01/FNS01_WOOD_Bullet_Impact_Ricochet_Med_Cue
root=QualityLevel_0
nodes [
 {id=Modulator_0, children [Random_0], properties {pitchMax=1.2, pitchMin=0.9, volumeMax=1.5}}
 {id=Modulator_2, children [Random_1], properties {pitchMax=1.2, pitchMin=0.9, volumeMax=1.5}}
 {id=QualityLevel_0, children [Modulator_0,Modulator_2,Modulator_2,Modulator_2]}
 {
  id=Random_0
  children [
   WavePlayer_0
   WavePlayer_1
   WavePlayer_2
   WavePlayer_3
   WavePlayer_4
   WavePlayer_5
   WavePlayer_6
   WavePlayer_7
   WavePlayer_8
   WavePlayer_9
   WavePlayer_10
   WavePlayer_15
   WavePlayer_17
   WavePlayer_19
  ]
  properties {weights [1,1,1,1,1,1,1,1,1,1,1,1,1,1]}
 }
 {
  id=Random_1
  children [WavePlayer_21, WavePlayer_22, WavePlayer_23, WavePlayer_25, WavePlayer_24, WavePlayer_26, WavePlayer_27, WavePlayer_28, WavePlayer_29]
  properties {weights [1,1,1,1,1,1,1,1,1]}
 }
 {id=WavePlayer_0, properties {soundWaveAssetPtr=/Game/Sounds/Fort_Weapons/Impacts/Bullet_Impacts_S01/Wavs/WOOD_IMPACT_MED_000}}
 {id=WavePlayer_1, properties {soundWaveAssetPtr=/Game/Sounds/Fort_Weapons/Impacts/Bullet_Impacts_S01/Wavs/WOOD_IMPACT_MED_001}}
 {id=WavePlayer_10, properties {soundWaveAssetPtr=/Game/Sounds/Fort_Weapons/Impacts/Bullet_Impacts_S01/Wavs/WOOD_IMPACT_MED_019}}
 {id=WavePlayer_15, properties {soundWaveAssetPtr=/Game/Sounds/Fort_Weapons/Impacts/Bullet_Impacts_S01/Wavs/WOOD_IMPACT_MED_021}}
 {id=WavePlayer_17, properties {soundWaveAssetPtr=/Game/Sounds/Fort_Weapons/Impacts/Bullet_Impacts_S01/Wavs/WOOD_IMPACT_MED_022}}
 {id=WavePlayer_19, properties {soundWaveAssetPtr=/Game/Sounds/Fort_Weapons/Impacts/Bullet_Impacts_S01/Wavs/WOOD_IMPACT_MED_023}}
 {id=WavePlayer_2, properties {soundWaveAssetPtr=/Game/Sounds/Fort_Weapons/Impacts/Bullet_Impacts_S01/Wavs/WOOD_IMPACT_MED_003}}
 {id=WavePlayer_21, properties {soundWaveAssetPtr=&WOOD_IMPACT_MED_000}}
 {id=WavePlayer_22, properties {soundWaveAssetPtr=&WOOD_IMPACT_MED_001}}
 {id=WavePlayer_23, properties {soundWaveAssetPtr=/Game/Sounds/Fort_Weapons/Impacts/Bullet_Impacts_S01/Wavs/WOOD_IMPACT_MED_006}}
 {id=WavePlayer_24, properties {soundWaveAssetPtr=/Game/Sounds/Fort_Weapons/Impacts/Bullet_Impacts_S01/Wavs/WOOD_IMPACT_MED_014}}
 {id=WavePlayer_25, properties {soundWaveAssetPtr=/Game/Sounds/Fort_Weapons/Impacts/Bullet_Impacts_S01/Wavs/WOOD_IMPACT_MED_013}}
 {id=WavePlayer_26, properties {soundWaveAssetPtr=&WOOD_IMPACT_MED_019}}
 {id=WavePlayer_27, properties {soundWaveAssetPtr=&WOOD_IMPACT_MED_021}}
 {id=WavePlayer_28, properties {soundWaveAssetPtr=&WOOD_IMPACT_MED_022}}
 {id=WavePlayer_29, properties {soundWaveAssetPtr=&WOOD_IMPACT_MED_023}}
 {id=WavePlayer_3, properties {soundWaveAssetPtr=&WOOD_IMPACT_MED_006}}
 {id=WavePlayer_4, properties {soundWaveAssetPtr=/Game/Sounds/Fort_Weapons/Impacts/Bullet_Impacts_S01/Wavs/WOOD_IMPACT_MED_008}}
 {id=WavePlayer_5, properties {soundWaveAssetPtr=/Game/Sounds/Fort_Weapons/Impacts/Bullet_Impacts_S01/Wavs/WOOD_IMPACT_MED_009}}
 {id=WavePlayer_6, properties {soundWaveAssetPtr=/Game/Sounds/Fort_Weapons/Impacts/Bullet_Impacts_S01/Wavs/WOOD_IMPACT_MED_011}}
 {id=WavePlayer_7, properties {soundWaveAssetPtr=&WOOD_IMPACT_MED_013}}
 {id=WavePlayer_8, properties {soundWaveAssetPtr=&WOOD_IMPACT_MED_014}}
 {id=WavePlayer_9, properties {soundWaveAssetPtr=/Game/Sounds/Fort_Weapons/Impacts/Bullet_Impacts_S01/Wavs/WOOD_IMPACT_MED_017}}
]
properties {
 attenuationSettings=/Game/Sounds/Attenuation/Fort_Bullet_Impact_Att
 maxDistance=5300
 soundClassObject=&Fort_Impacts
 volumeMultiplier=1.416
 concurrencySet [&RifleBulletImpact_Conc]
}
